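# Masked unlit Niagara sprite material for the Elegant Lily guitar effects, with a dissolve.
# Color comes from the texture's red channel times particle color, scaled by time-of-day brightness parameters and run through MF_FX_EmissiveAdjustments.
# Texture and particle alpha multiply a hard-edged mask, thresholded from a tiled dissolve texture by the Dissolve dynamic parameter and bypassed at values of one or above.
# A separate Dissolve scalar parameter is left unconnected.
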Material /SparksCosmetics/Instrument/Guitar/ElegantLily/Effects/M_Guitar_ElegantLily
graph {
 properties {
  bAutomaticallySetUsageInEditor=false
  bCastRayTracedShadows=false
  bUsedWithNiagaraSprites=true
  blendMode=BLEND_Masked
  shadingModel=MSM_Unlit
  usedShadingModels=Unlit
 }
 nodes [
  n1 = ScalarParameter() {defaultValue=0.9, parameterName=Morn}
  n2 = ScalarParameter() {defaultValue=1, parameterName=Day}
  n3 = ScalarParameter() {defaultValue=0.9, parameterName=Evening}
  n4 = ScalarParameter() {defaultValue=0.6, parameterName=Night}
  n5 = ParticleColor()
  n6 = TextureSampleParameter2D() {parameterName=Texture, texture=&/T_ElegantLily_Guitar}
  MaterialFunctionCall(Emissive=Multiply(
    A=MaterialFunctionCall(Color_Morning=@n1, Color_Day=@n2, Color_Evening=@n3, Color_Night=@n4) {materialFunction=&Color_Time_of_Day}.Result
    B=Multiply(A=@n5.RGB, B=@n6.R)
   )) {materialFunction=&MF_FX_EmissiveAdjustments}
  n7 = Constant() {r=1}
  n8 = DynamicParameter() {paramNames [Dissolve,Param2,Param3,Param4]}
  n9 = Add(A=OneMinus(None=@n8.Dissolve)) {constB=0.015}
  n10 = TextureCoordinate() {uTiling=3, vTiling=3}
  n11 = SmoothStep(
   Min=@n9
   Max=@n9
   Value=LinearInterpolate(Alpha=TextureSample(UVs=@n10) {texture=&/T_ElegantLily_Guitar_Dissolve}.RGB) {constA=0.1}
  )
  Multiply(A=@n5.A, B=Multiply(A=@n6.A, B=If(A=@n8.Dissolve, B=@n7, "A > B"=@n7, "A == B"=@n7, "A < B"=@n11)))
  ScalarParameter() {defaultValue=0.1, parameterName=Dissolve}
 ]
}
