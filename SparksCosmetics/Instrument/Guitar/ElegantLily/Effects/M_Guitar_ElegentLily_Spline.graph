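# Unlit masked Niagara particle material for the Elegant Lily guitar, drawing heart textures on a spline-thickened mesh.
# Dynamic parameters drive width, tapered along inverted V through SplineThicken, plus texture thickness and pan.
# Pan snaps to a Frames count for a stepped flipbook-like scroll of the center-scaled, clamped heart texture.
# Particle color tints the emissive, while blurred noise roughens a V-gradient mask multiplied into texture alpha.
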Material /SparksCosmetics/Instrument/Guitar/ElegantLily/Effects/M_Guitar_ElegentLily_Spline
graph {
 properties {
  bAutomaticallySetUsageInEditor=false
  bUsedWithNiagaraMeshParticles=true
  bUsedWithNiagaraSprites=true
  blendMode=BLEND_Masked
  shadingModel=MSM_Unlit
  usedShadingModels=Unlit
 }
 nodes [
  n1 = DynamicParameter() {defaultValue=<rgba 1 1 0.2 7>, paramNames [TexturePan, TextureThickness, ColorFade, SplineThickness]}
  n2 = ComponentMask(None=OneMinus(None=TextureCoordinate())) {g=true}
  n3 = LinearInterpolate(B=@n1.SplineThickness, Alpha=Power(Base=@n2)) {constB=-6}
  MaterialFunctionCall(WidthBase=@n3, WidthTip=@n3) {materialFunction=&SplineThicken}
  n4 = AppendVector(A=@n1.TextureThickness, B=Constant() {desc=Scruntch texture so it is the right proportion, r=0.2})
  n5 = Constant()
  n6 = ScalarParameter() {defaultValue=4, parameterName=Frames}
  n7 = Divide(A=Ceil(None=Multiply(A=@n1.TexturePan, B=@n6)), B=@n6)
  n8 = TextureSample(UVs=Add(
    A=MaterialFunctionCall(Texture Scale=@n4) {materialFunction=&ScaleUVsByCenter}.UVs
    B=AppendVector(A=@n5, B=@n7)
   ) {constB=0}) {samplerSource=SSM_Clamp_WorldGroupSettings, texture=&/T_ElegantLily_Guitar_Heart_Empty}
  MaterialFunctionCall(Emissive=Multiply(A=ParticleColor().RGB, B=@n8.RGB)) {materialFunction=&MF_FX_EmissiveAdjustments}
  n9 = TextureSample() {samplerType=SAMPLERTYPE_Grayscale, texture=&LowResBlurredNoise}
  n10 = Subtract(A=Add(A=@n2, B=Multiply(A=@n2, B=@n9.R)), B=Constant() {r=0.6})
  n11 = Constant() {r=0.1}
  Multiply(A=@n8.A, B=Saturate(None=Divide(A=@n10, B=@n11)))
  Reroute(None=@n1.ColorFade)
 ]
}
